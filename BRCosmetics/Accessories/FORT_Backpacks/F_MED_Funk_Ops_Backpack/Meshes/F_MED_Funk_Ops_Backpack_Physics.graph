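# Physics asset for the Funk Ops backpack cosmetic, defining the bodies and joints for its disco ball.
# A large sphere covers the ball, while tiny capsules stand in for the intermediate rotation bone and the root.
# Two constraints form a chain: the rotation bone turns freely on all three angular axes relative to root.
# The ball is free only on swing 1 relative to that bone, giving it a hinge-like joint.
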PhysicsAsset /BRCosmetics/Accessories/FORT_Backpacks/F_MED_Funk_Ops_Backpack/Meshes/F_MED_Funk_Ops_Backpack_Physics
bodies {
 dyn_disco_ball [{shapeType=Sphere, radius=21.7614}]
 dyn_disco_ball_rot [{shapeType=Capsule, length=0.505, radius=0.505, rotation=<rot 0 -90 0>}]
 root [{shapeType=Capsule, length=0.505, radius=0.505, rotation=<rot 0 -90 0>}]
}
constraints [
 {bone1Name=dyn_disco_ball, bone2Name=dyn_disco_ball_rot, swing1LimitDegrees=45, swing1Motion=Free, swing2LimitDegrees=0, twistLimitDegrees=0}
 {
  bone1Name=dyn_disco_ball_rot
  bone2Name=root
  swing1LimitDegrees=45
  swing1Motion=Free
  swing2LimitDegrees=45
  swing2Motion=Free
  twistLimitDegrees=45
  twistMotion=Free
 }
]
