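# Animation Blueprint for the Funk Ops backpack cosmetic, adding physics motion and a spinning disco ball.
# A RigidBody node simulates the reference pose in base-bone space (root) with overridden gravity, then feeds an additive Transform Bone node on dyn_disco_ball_rot.
# That node offsets the bone slightly and applies a yaw that the event graph decrements by a fixed step each update, so spin speed is frame-rate dependent.
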AnimBlueprint /BRCosmetics/Accessories/FORT_Backpacks/F_MED_Funk_Ops_Backpack/Meshes/F_MED_Funk_Ops_Backpack_AnimBP
target_skeleton=/Game/Accessories/FORT_Backpacks/F_MED_Funk_Ops_Backpack/Meshes/F_MED_Funk_Ops_Backpack_Skeleton
graphs [
 AnimGraph {nodes [
   n1 = RigidBody(ComponentPose=Local To Component(LocalPose=Local Space Ref Pose())) {
    bOverrideWorldGravity=true
    overridePhysicsAsset=&/F_MED_Funk_Ops_Backpack_Physics
    simulationSpace=BaseBoneSpace
    alphaBoolBlend {blendOption=Linear}
    baseBoneRef {boneName=root}
    componentAppliedLinearAccClamp=<2000 2000 2000>
    componentLinearAccScale=<1 1 1>
    componentLinearVelScale=<1 1 1>
    overrideWorldGravity {z=-980}
   }
   n2 = "Transform (Modify) Bone - Bone: dyn_disco_ball_rot"(ComponentPose=@n1, Alpha=1, Translation="0,1.5,0", Rotation=Get disco_ball() -> (disco_ball:Rotator), Scale="1,1,1") {
    rotationMode=BMM_Additive
    translationMode=BMM_Additive
    translationSpace=BCS_BoneSpace
    alphaBoolBlend {blendOption=Linear}
    boneToModify {boneName=dyn_disco_ball_rot}
    translation {y=1.5}
   }
   Output Pose(Result=Component To Local(ComponentPose=@n2)) {layerGroup=None}
  ]}
 EventGraph {nodes [
   Event BlueprintUpdateAnimation()
   > Set rotation(rotation:double=float - float(A=Get rotation() -> (rotation:double), B=2) {is=title K2Node_CallFunction}) -> (Output_Get:double)
   > Set disco_ball(disco_ball:Rotator=MakeRotator(Yaw=Get rotation())) -> (Output_Get:Rotator)
   TryGetPawnOwner()
  ]}
]
variables [
 disco_ball {type=Rotator}
 rotation {type=number}
]
